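# Additive unlit material for Niagara mesh particles, named as a temporary glow, with a graph mostly made up of Valet turret vertex-animation logic.
# Vertex-color masks and pivot rotations build offsets for kickback, pitch, yaw, three-leg sway, activation and weight shift, chained through named reroutes.
# Curve-atlas rows shape motion from time and custom primitive data, and FixRotateAboutAxisNormals recomputes rotated normals.
# A camera-offset branch adds animation offset scaled by AnimWPO.
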Material /VehicleMods/FX/Systems/Generic/M_Temp_Glow
graph {
 properties {
  bAutomaticallySetUsageInEditor=false
  bCastRayTracedShadows=false
  bUsedWithNiagaraMeshParticles=true
  blendMode=BLEND_Additive
  pixelDepthOffsetMode=PDOM_Legacy
  shadingModel=MSM_Unlit
  usedShadingModels=Unlit
 }
 nodes [
  n1 = VertexNormalWS()
  n2 = ComponentMask(None=NamedRerouteUsage()) {b=true}
  n3 = Time()
  n4 = ScalarParameter() {defaultValue=0.125, parameterName=AmbientLegMovementSpeed}
  n5 = ConstantBiasScale(None=Sine(None=Multiply(A=@n3, B=@n4) {constB=0.25}))
  n6 = CurveAtlasRowParameter(CurveTime=@n5) {
   atlas=/Game/Vehicles/Valet/Textures/Turret_Atlas
   curve=/Game/Vehicles/Valet/Textures/LegAmbientAnimCurve
   defaultValue=2
   parameterName=LegAnimCurve
  }
  n7 = ChannelMaskParameter(None=NamedRerouteUsage()) {parameterName=LegVertColorMask}
  n8 = ComponentMask(None=NamedRerouteUsage()) {g=true}
  n9 = LinearInterpolate(
   A=Multiply(
    A=MaterialFunctionCall(A=@n6.R, B=@n6.R, C=@n6.R, Alpha=@n7.R) {materialFunction=&Lerp_3Color}.Result
    B=ScalarParameter() {defaultValue=0.1, parameterName=LegMoveStr}
   ) {constB=0.1}
   B=Multiply(A=ScalarParameter() {defaultValue=0.07, parameterName=DebugInactiveFootRotation}, B=@n8)
   Alpha=Ceil(None=@n8)
  )
  n10 = MaterialFunctionCall!(Rotation Axis=NamedRerouteUsage(), Rotation Angle=@n9, World Space Vertex Normals=@n1, Movement Mask=@n2, PivotPoint) {materialFunction=/Game/Vehicles/Valet/FixRotateAboutAxisNormals}
  n11 = Transform(None=Constant3Vector() {constant=<rgba 0 0 1 0>}) {transformSourceType=TRANSFORMSOURCE_Local}
  n12 = CurveAtlasRowParameter(CurveTime=@n5) {
   atlas=&Turret_Atlas
   curve=/Game/Vehicles/Valet/Textures/LegAmbientAnimCurve3
   defaultValue=4
   parameterName=LegAnimCurve2
  }
  n13 = NamedRerouteUsage()
  n14 = Multiply(
   A=ScalarParameter() {defaultValue=0.15, parameterName=LegMoveStr_02}
   B=MaterialFunctionCall(A=@n12.R, B=@n12.R, C=@n12.R, Alpha=ChannelMaskParameter(None=@n13) {parameterName=LegVertColorMask}.R) {materialFunction=&Lerp_3Color}.Result
  )
  Add(
   A=@n10.RotationOutputOnly
   B=MaterialFunctionCall(Rotation Axis=@n11, Rotation Angle=@n14, World Space Vertex Normals=Add(A=@n10.RotationOutputOnly, B=@n1), Movement Mask=@n2) {materialFunction=&FixRotateAboutAxisNormals}.RotationOutputOnly
  )
  n15 = WorldPosition()
  n16 = NamedRerouteUsage()
  n17 = RotateAboutAxis(NormalizedRotationAxis=NamedRerouteUsage(), RotationAngle=@n9, PivotPoint=@n16, Position=@n15.XYZ)
  Add(A=@n17, B=RotateAboutAxis(NormalizedRotationAxis=@n11, RotationAngle=@n14, PivotPoint=@n16, Position=Add(A=@n17, B=@n15.XYZ)))
  n18 = ScalarParameter() {defaultValue=1}
  Add(
   A=MaterialFunctionCall(Offset Amount=@n18) {materialFunction=/Engine/Functions/Engine_MaterialFunctions02/WorldPositionOffset/CameraOffset}.Clamped Camera Offset
   B=Multiply(A=ScalarParameter() {parameterName=AnimWPO}, B=NamedRerouteUsage())
  )
  Constant3Vector()
  n19 = Time()
  ConstantBiasScale(None=Sine(None=Multiply(A=@n19)))
  Constant()
  Frac(None=Multiply(A=Time(), B=ScalarParameter() {defaultValue=2, parameterName=DebugSpeed}))
  MaterialFunctionCall() {materialFunction=&ObjectPivotPoint}
  Multiply(A=VertexNormalWS(), B=@n18)
  NamedRerouteDeclaration(None=VertexColor()) {name=VertexColor, nodeColor=<rgba 0.8516 1 0 1>}
  n20 = Add(A=NamedRerouteUsage(), B=NamedRerouteUsage())
  n21 = Add(
   A=VertexNormalWS()
   B=Add(A=NamedRerouteDeclaration(None=@n20) {name=AN_Kickback+Pitch_Normal, nodeColor=<rgba 1 0.375 0 1>}, B=NamedRerouteUsage())
  )
  NamedRerouteDeclaration(None=Transform(None=Normalize(VectorInput=@n21)) {transformSourceType=TRANSFORMSOURCE_World, transformType=TRANSFORM_Tangent}) {name=RecalculatedNormals, nodeColor=<rgba 0.5938 1 0 1>}
  n22 = Saturate(None=ScalarParameter() {bUseCustomPrimitiveData=true, parameterName=W_WeightShift, primitiveDataIndex=4})
  NamedRerouteDeclaration(None=Multiply(A=@n22, B=ScalarParameter() {defaultValue=-0.05, parameterName=WeightShift_Amt})) {name=DeltaTimeRotation, nodeColor=<rgba 1 0.6328 0 1>}
  n23 = NamedRerouteUsage()
  NamedRerouteDeclaration(None=SphereMask(A=ComponentMask(None=@n23) {g=true}, B=Constant() {r=0.5}) {attenuationRadius=0.1}) {name=NoYawMask, nodeColor=<rgba 0 1 0.25 1>}
  n24 = ComponentMask(None=NamedRerouteUsage()) {b=true}
  n25 = SphereMask(A=ChannelMaskParameter(None=NamedRerouteUsage()) {parameterName=LegVertColorMask}.R, B=NamedRerouteUsage(), Radius=Constant() {r=0.2})
  NamedRerouteDeclaration(None=OneMinus(None=Multiply(A=@n24, B=@n25))) {name=WeightShiftMask, nodeColor=<rgba 1 0 0.2578 1>}
  n26 = NamedRerouteUsage()
  n27 = NamedRerouteUsage()
  n28 = NamedRerouteUsage()
  n29 = Add(A=WorldPosition().XYZ, B=NamedRerouteUsage())
  NamedRerouteDeclaration(None=Multiply(A=RotateAboutAxis(NormalizedRotationAxis=@n26, RotationAngle=@n27, PivotPoint=@n28, Position=@n29), B=NamedRerouteUsage())) {name=WeightShiftWPO, nodeColor=<rgba 0 1 0.9297 1>}
  NamedRerouteDeclaration(None=VectorParameter() {parameterName=Leg3_PivotPosition, defaultValue=<rgba -9.3805 -18.1743 8.3615 1>}.RGB) {name=Leg3PivotPosition, nodeColor=<rgba 1 0 0.2109 1>}
  NamedRerouteDeclaration(None=VectorParameter() {parameterName=Leg2_PivotPosition, defaultValue=<rgba -9.3805 19.0524 8.3615 1>}.RGB) {name=Leg2PivotPosition, nodeColor=<rgba 1 0.6562 0 1>}
  NamedRerouteDeclaration(None=VectorParameter() {parameterName=Leg1_PivotPosition, defaultValue=<rgba 21.3263 1.3855 8.3615 1>}.RGB) {name=Leg1PivotPosition, nodeColor=<rgba 0 1 1 1>}
  NamedRerouteDeclaration(None=VectorParameter() {parameterName=Leg3RotationAxis, defaultValue=<rgba -0.838 0.546 0 1>}.RGB) {name=Leg3RotationAxis, nodeColor=<rgba 1 0.8672 0 1>}
  NamedRerouteDeclaration(None=VectorParameter() {parameterName=Leg2RotationAxis, defaultValue=<rgba 0.887 0.463 0 1>}.RGB) {name=Leg2RotationAxis, nodeColor=<rgba 0 1 0.6016 1>}
  NamedRerouteDeclaration(None=@n17) {name=AmbientLegMovement, nodeColor=<rgba 1 0 0.9844 1>}
  NamedRerouteDeclaration(None=VectorParameter() {parameterName=Leg1RotationAxis, defaultValue=<rgba 0 -1 0 1>}.RGB) {name=Leg1RotationAxis, nodeColor=<rgba 0 1 0.8125 1>}
  n30 = VertexNormalWS()
  n31 = NamedRerouteUsage()
  NamedRerouteDeclaration(None=MaterialFunctionCall(
    Rotation Axis=NamedRerouteUsage()
    Rotation Angle=NamedRerouteUsage()
    World Space Vertex Normals=Add(A=@n30, B=@n31)
    Movement Mask=NamedRerouteUsage()
   ) {materialFunction=&FixRotateAboutAxisNormals}.RotationOutputOnly) {name=WeightShiftWPO_Normals, nodeColor=<rgba 0 1 0.1562 1>}
  n32 = Subtract(A=WorldPosition() {worldPositionShaderOffset=WPT_ExcludeAllShaderOffsets}.XYZ, B=ObjectPositionWS())
  n33 = VectorParameter() {parameterName=Offset_TEmp, defaultValue=<rgba 0 0 -5 1>}
  n34 = Saturate(None=ScalarParameter() {bUseCustomPrimitiveData=true, parameterName=W_Activate, primitiveDataIndex=3})
  n35 = NamedRerouteUsage()
  n36 = CurveAtlasRowParameter(CurveTime=NamedRerouteDeclaration(None=@n22) {name=WeightShiftAnimCurve, nodeColor=<rgba 0.0547 0 1 1>}) {
   atlas=&Turret_Atlas
   curve=/Game/Vehicles/Valet/Textures/TurretActivateAnimCurve
   defaultValue=5
   parameterName=ActivateAnimCurve
  }
  n37 = Max(A=Multiply(A=NamedRerouteUsage(), B=OneMinus(None=@n34)), B=Max(A=ComponentMask(None=@n35) {g=true}, B=@n36.R))
  NamedRerouteDeclaration(None=Add(A=Normalize(VectorInput=@n32), B=Multiply(A=@n33.RGB, B=@n37))) {name=Activate_WPO, nodeColor=<rgba 0.6172 1 0 1>}
  n38 = NamedRerouteUsage()
  n39 = ComponentMask(None=@n38) {g=true}
  n40 = Max(A=@n39, B=ComponentMask(None=@n38) {r=true})
  NamedRerouteDeclaration(None=Multiply(A=VertexNormalWS(), B=@n40)) {name=AN_Kickback_Normal, nodeColor=<rgba 0 0.4141 1 1>}
  n41 = WorldPosition()
  n42 = MaterialFunctionCall() {materialFunction=&ObjectPivotPoint}
  n43 = VectorParameter() {parameterName=W_DirectionVector, defaultValue=<rgba 1 0 0 1>}
  n44 = LinearInterpolate(
   A=ScalarParameter() {defaultValue=4, parameterName=HullKickbackIntensity}
   B=ScalarParameter() {defaultValue=16, parameterName=KickbackIntensity}
   Alpha=@n39
  )
  n45 = ScalarParameter() {bUseCustomPrimitiveData=true, parameterName=W_KickBack, primitiveDataIndex=1}
  n46 = Divide(A=Subtract(A=Time(), B=@n45), B=ScalarParameter() {defaultValue=0.25, parameterName=W_AnimDuration})
  n47 = CurveAtlasRowParameter(CurveTime=StaticSwitchParameter(True=Saturate(None=@n46), False=@n45) {parameterName=UseTimeStamp}) {atlas=&Turret_Atlas, curve=/Game/Vehicles/Valet/Textures/MachineGunAnimCurve, parameterName=Turret_AnimCurve}
  n48 = Subtract(A=Multiply(A=@n47.R) {constB=2})
  n49 = Add(
   A=Normalize(VectorInput=Subtract(A=@n41.XYZ, B=@n42.Object Pivot Location))
   B=Multiply(A=Transform(None=@n43.RGB) {transformSourceType=TRANSFORMSOURCE_Local}, B=Multiply(A=@n44, B=@n48))
  )
  NamedRerouteDeclaration(None=Add(A=Multiply(A=@n49, B=@n40), B=NamedRerouteUsage())) {name=AN_Kickback_WPO, nodeColor=<rgba 1 0.7266 0 1>}
  n50 = Transform(None=VectorParameter() {parameterName=W_LocalPitchAxis, defaultValue=<rgba 0 1 0 1>}.RGB) {transformSourceType=TRANSFORMSOURCE_Local}
  n51 = LinearInterpolate(
   A=ScalarParameter() {bUseCustomPrimitiveData=true, parameterName=W_Pitch}
   B=ScalarParameter() {defaultValue=-0.15, parameterName=InactivePitchValue}
   Alpha=ComponentMask(None=NamedRerouteUsage()) {r=true}
  )
  n52 = MaterialFunctionCall() {materialFunction=&ObjectPivotPoint}
  n53 = Transform(None=VectorParameter() {parameterName=W_PitchPivotOffset, defaultValue=<rgba -10.5986 0 39.7765 1>}.RGB) {transformSourceType=TRANSFORMSOURCE_Local}
  n54 = NamedRerouteUsage()
  n55 = WorldPosition()
  NamedRerouteDeclaration(None=RotateAboutAxis(NormalizedRotationAxis=@n50, RotationAngle=@n51, PivotPoint=Add(A=@n52.Object Pivot Location, B=@n53), Position=Add(A=@n54, B=@n55.XYZ))) {name=AN_Pitch_WPO, nodeColor=<rgba 0 1 0.4375 1>}
  n56 = NamedRerouteUsage()
  n57 = VertexNormalWS()
  n58 = NamedRerouteUsage()
  NamedRerouteDeclaration(None=MaterialFunctionCall(Rotation Axis=@n50, Rotation Angle=@n51, World Space Vertex Normals=Add(A=@n56, B=@n57), Movement Mask=ComponentMask(None=@n58) {r=true}) {materialFunction=&FixRotateAboutAxisNormals}.RotationOutputOnly) {name=AN_Pitch_Normal, nodeColor=<rgba 0.3594 0 1 1>}
  n59 = Add(A=VertexNormalWS(), B=@n10.RotationOutputOnly)
  NamedRerouteDeclaration(None=Transform(None=Normalize(VectorInput=@n59)) {transformSourceType=TRANSFORMSOURCE_World, transformType=TRANSFORM_Tangent}) {name=AmbientLegMovementVertexNormal, nodeColor=<rgba 0.6406 1 0 1>}
  n60 = Transform(None=VectorParameter() {parameterName=W_LocalYawAxis, defaultValue=<rgba 0 0 1 1>}.RGB) {transformSourceType=TRANSFORMSOURCE_Local}
  n61 = LinearInterpolate(
   A=ScalarParameter() {bUseCustomPrimitiveData=true, parameterName=W_Yaw, primitiveDataIndex=2}
   B=ScalarParameter() {parameterName=InactiveYawValue}
   Alpha=ComponentMask(None=NamedRerouteUsage()) {r=true}
  )
  n62 = Add(
   A=MaterialFunctionCall() {materialFunction=&ObjectPivotPoint}.Object Pivot Location
   B=Transform(None=VectorParameter() {parameterName=W_YawPivotOffset, defaultValue=<rgba 0 0 0 1>}.RGB) {transformSourceType=TRANSFORMSOURCE_Local}
  )
  n63 = Add(A=NamedRerouteUsage(), B=WorldPosition().XYZ)
  n64 = OneMinus(None=NamedRerouteUsage())
  NamedRerouteDeclaration(None=Multiply(A=RotateAboutAxis(NormalizedRotationAxis=@n60, RotationAngle=@n61, PivotPoint=@n62, Position=@n63), B=@n64)) {name=AN_Yaw_WPO, nodeColor=<rgba 1 0.3984 0 1>}
  n65 = Add(A=NamedRerouteUsage(), B=VertexNormalWS())
  NamedRerouteDeclaration(None=Multiply(
    A=MaterialFunctionCall(Rotation Axis=@n60, Rotation Angle=@n61, World Space Vertex Normals=@n65) {materialFunction=&FixRotateAboutAxisNormals}.RotationOutputOnly
    B=@n64
   )) {name=AN_Yaw_Normal, nodeColor=<rgba 0 1 0.1328 1>}
  n66 = NamedRerouteUsage()
  n67 = NamedRerouteUsage()
  n68 = NamedRerouteUsage()
  n69 = NamedRerouteDeclaration(None=Multiply(A=Add(A=@n66, B=@n67), B=ComponentMask(None=@n68) {r=true}) {constB=1.008}) {name=AN_Kickback+Pitch_WPO, nodeColor=<rgba 1 0 0.5625 1>}
  n70 = NamedRerouteUsage()
  n71 = NamedRerouteUsage()
  n72 = NamedRerouteDeclaration(None=LinearInterpolate(A=Add(A=@n69, B=@n70), B=NamedRerouteUsage(), Alpha=ComponentMask(None=@n71) {b=true})) {name=WPO_Rotations, nodeColor=<rgba 0 1 0.3203 1>}
  n73 = NamedRerouteUsage()
  NamedRerouteDeclaration(None=Add(A=Add(A=@n72, B=@n73), B=NamedRerouteUsage())) {name=FinalWPO, nodeColor=<rgba 1 0.7969 0 1>}
  n74 = LinearInterpolate(A=NamedRerouteUsage(), B=NamedRerouteUsage(), Alpha=ComponentMask(None=NamedRerouteUsage()) {b=true})
  n75 = VertexNormalWS()
  NamedRerouteDeclaration(None=LinearInterpolate(
    A=NamedRerouteDeclaration(None=@n74) {name=WPO_Normals, nodeColor=<rgba 1 0 0.8203 1>}
    B=Transform(None=@n75) {transformSourceType=TRANSFORMSOURCE_World, transformType=TRANSFORM_Tangent}
    Alpha=NamedRerouteUsage()
   ) {constB=0}) {name=Final_Normal, nodeColor=<rgba 0 0.9297 1 1>}
  n76 = MaterialFunctionCall(Float3=NamedRerouteUsage()) {materialFunction=&BreakOutFloat3Components}
  NamedRerouteDeclaration(None=Multiply(A=@n76.R, B=@n76.R)) {name=BarrelMask, nodeColor=<rgba 1 0.5625 0 1>}
  NamedRerouteDeclaration(None=@n6.A) {name=AmbientSquatCurve, nodeColor=<rgba 0.8281 1 0 1>}
  n77 = NamedRerouteUsage()
  n78 = NamedRerouteUsage()
  n79 = NamedRerouteUsage()
  n80 = ChannelMaskParameter(None=NamedRerouteUsage()) {parameterName=LegVertColorMask}
  NamedRerouteDeclaration(None=TransformPosition(None=MaterialFunctionCall(A=@n77, B=@n78, C=@n79, Alpha=@n80.R) {materialFunction=&Lerp_3Color}.Result) {transformType=TRANSFORMPOSSOURCE_World}) {name=LegPivotPositions, nodeColor=<rgba 0.4062 0 1 1>}
  n81 = NamedRerouteUsage()
  n82 = NamedRerouteUsage()
  n83 = NamedRerouteUsage()
  n84 = ChannelMaskParameter(None=NamedRerouteUsage()) {parameterName=LegVertColorMask}
  NamedRerouteDeclaration(None=Transform(None=MaterialFunctionCall(A=@n81, B=@n82, C=@n83, Alpha=@n84.R) {materialFunction=&Lerp_3Color}.Result) {transformSourceType=TRANSFORMSOURCE_Local}) {name=LegsDirectionVector, nodeColor=<rgba 0 0.3672 1 1>}
  n85 = CurveAtlasRowParameter(CurveTime=@n34) {
   atlas=&Turret_Atlas
   curve=&TurretActivateAnimCurve
   defaultValue=5
   parameterName=ActivateAnimCurve
  }
  NamedRerouteDeclaration(None=AppendVector(A=@n85, B=@n85)) {name=ActivateAnimCurves, nodeColor=<rgba 1 0.1875 0 1>}
  NamedRerouteDeclaration(None=ScalarParameter() {defaultValue=0.5, parameterName=PickLeg}) {name=Weightedleg, nodeColor=<rgba 1 0 0.4688 1>}
  n86 = NamedRerouteUsage()
  n87 = NamedRerouteUsage()
  n88 = NamedRerouteUsage()
  n89 = NamedRerouteUsage()
  NamedRerouteDeclaration(None=Transform(None=MaterialFunctionCall(A=@n86, B=@n87, C=@n88, Alpha=@n89) {materialFunction=&Lerp_3Color}.Result) {transformSourceType=TRANSFORMSOURCE_Local}) {name=WeightedLegAxis, nodeColor=<rgba 1 0 0.6094 1>}
  n90 = NamedRerouteUsage()
  n91 = NamedRerouteUsage()
  n92 = NamedRerouteUsage()
  n93 = NamedRerouteUsage()
  NamedRerouteDeclaration(None=TransformPosition(None=MaterialFunctionCall(A=@n90, B=@n91, C=@n92, Alpha=@n93) {materialFunction=&Lerp_3Color}.Result) {transformType=TRANSFORMPOSSOURCE_World}) {name=WeightedLegPosition, nodeColor=<rgba 1 0.1172 0 1>}
  NamedRerouteUsage()
  ObjectPositionWS()
  ParticleColor()
 ]
}
